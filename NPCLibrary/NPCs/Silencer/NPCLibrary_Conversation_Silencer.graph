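# Conversation database for the Silencer NPC, defining the greeting players get when they talk to the character.
# The Silencer entry point leads to one speech node with a default High Stakes Club line.
# Contextual lines replace it for a first-ever conversation or when the player wears specific outfits, ranked by curve-driven priorities.
# The speech node then links out to the shared NPCLibrary quest and basic services conversation entries.
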ConversationDatabase /NPCLibrary/NPCs/Silencer/NPCLibrary_Conversation_Silencer
entry_points [{tag=Conversation.Entry.Silencer, destinations [#1]}]
nodes [
 ConversationEntryPointNode_7 {
  id=#1
  class=ConversationEntryPointNode
  outputs [#2]
  properties {entryTag {tagName=Conversation.Entry.Silencer}}
 }
 Speech_C_4 {
  id=#2
  class=Speech_C
  outputs [#3, #4]
  properties {
   nodeName=""
   generalConfig {
    defaultMessage=What can The High Stakes Club do for you?
    contextualMessages [
     {
      message=You ever find out that your idol is a literal monster? It's... not fun.
      contextRequirements [{requirement {class=FortControllerRequirement_IsFirstEverConversationWithNPC}}]
      priority {curve {rowName=Tandem.Speech.Priority.FirstEverInteraction}}
     }
     {
      message=You've ruined my favorite album forever. You truly are a monster.
      contextRequirements [{requirement {class=FortControllerRequirement_HasCID, properties {allowedCIds [/BRCosmetics/Athena/Items/Cosmetics/Characters/Character_IchorIncisor]}}}]
      priority {curve {rowName=Tandem.Speech.Priority.OutfitReaction}}
     }
     {
      message=I just wanted ONE THING that made me feel like everyone else. The Countess ruined that.
      contextRequirements [{requirement {class=FortControllerRequirement_HasCID, properties {allowedCIds [&Character_SharpFang]}}}]
      priority {curve {rowName=Tandem.Speech.Priority.OutfitReaction}}
     }
     {
      message=We have to take down Countess Daraku. Who knows how many fans she's fed on?
      contextRequirements [{requirement {class=FortControllerRequirement_HasCID, properties {allowedCIds [/BRCosmetics/Athena/Items/Cosmetics/Characters/Character_Sparrow]}}}]
      priority {curve {rowName=Tandem.Speech.Priority.OutfitReaction}}
     }
     {
      message=I can't hold myself back anymore. Show me how to give everything to the hunt.
      contextRequirements [{requirement {
         class=FortControllerRequirement_HasCID
         properties {allowedCIds [/BRCosmetics/Athena/Items/Cosmetics/Characters/CID_869_Athena_Commando_F_Cavalry]}
        }}]
      priority {curve {rowName=Tandem.Speech.Priority.OutfitReaction}}
     }
    ]
   }
  }
 }
 ConversationLinkNode_0 {id=#4, class=ConversationLinkNode, properties {remoteEntryTag {tagName=Conversation.Entry.BasicServices}}}
 ConversationLinkNode_1 {id=#3, class=ConversationLinkNode, properties {remoteEntryTag {tagName=Conversation.Entry.NPCLibrary.Quests}}}
]
